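# Unlit, two-sided material for the projection lines of the LimeRock countdown timer hologram, using the shared hologram function for color, mask and WPO.
# A UV1 gradient and collapse parameter blend vertices between the actor origin and hologram WPO.
# Opacity combines gradient masks with a noise-modulated sphere band glow, scaled by camera-depth terms whose lerps have identical endpoints.
# A Debug blend, defaulting to 1, swaps in white emissive with plain hologram WPO.
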Material /LimeRock_CountdownTimer/Gameplay/CountdownTimer/FX/Timer/Mats/M_LimeRock_TimerProjectionLines
graph {
 properties {
  bAutomaticallySetUsageInEditor=false
  bUseMaterialAttributes=true
  pixelDepthOffsetMode=PDOM_Legacy
  shadingModel=MSM_Unlit
  twoSided=true
  usedShadingModels=Unlit
 }
 nodes [
  n1 = VectorParameter() {parameterName=Final color HSV adjustment, channelNames {a=Na, b=V, g=S, r=H}, defaultValue=<rgba 0 1 1 0>}
  n2 = MaterialFunctionCall() {materialFunction=/LimeRock_CountdownTimer/Gameplay/CountdownTimer/MF_LimeRockCountdownHologram}
  n3 = Desaturation(
   None=MaterialFunctionCall(Hue Shift Percentage=@n1.H, Texture=@n2.Color) {materialFunction=&HueShift}.Result
   Fraction=OneMinus(None=@n1.S)
  )
  n4 = Multiply(
   A=AppendVector(
    A=MaterialFunctionCall() {materialFunction=&LinearGradient}.UGradient
    B=Constant() {r=0.5}
   )
   B=ScalarParameter() {defaultValue=0.75, parameterName=Tile}
  )
  n5 = Multiply(A=Time(), B=ScalarParameter() {defaultValue=0.3, parameterName=Speed})
  n6 = Panner(Coordinate=@n4, Time=@n5) {speedX=1, speedY=0.1}
  n7 = Panner(Coordinate=Multiply(A=@n4) {constB=0.75}, Time=@n5) {speedX=-1, speedY=-0.1}
  n8 = LinearInterpolate(Alpha=Min(
    A=TextureSample(UVs=@n6) {constCoordinate=1, samplerType=SAMPLERTYPE_LinearColor, texture=/EpicBaseTextures/Noises/T_Perlin_Noise_D}.R
    B=TextureSample(UVs=@n7) {constCoordinate=1, samplerType=SAMPLERTYPE_LinearColor, texture=&T_Perlin_Noise_D}.R
   )) {constA=0.1}
  n9 = ScalarParameter() {defaultValue=250, parameterName=Band Radius}
  n10 = MaterialFunctionCall!(
   Width=CollectionParameter() {collection=/LimeRock_CountdownTimer/Gameplay/CountdownTimer/MPC_CountdownTimerVisuals, parameterName=Timer_Lines_BottomGlowWidth}
   Sharpness=ScalarParameter() {defaultValue=0.9, parameterName=Band sharpness}
   Offset
   Direction Switch=StaticBool() {value=true}
   Compare
   Input Coordinates
  ) {materialFunction=/Engine/Functions/Engine_MaterialFunctions03/Procedurals/GeneratedBand}
  n11 = MaterialFunctionCall(
   Width=ScalarParameter() {defaultValue=0.18, parameterName=Band with 2}
   Sharpness=ScalarParameter() {defaultValue=0.9, parameterName=Band sharpness 2}
  ) {materialFunction=&GeneratedBand}
  n12 = SphereMask(
   A=Add(A=NamedRerouteUsage(), B=WorldPosition().XYZ)
   B=ActorPositionWS()
   Radius=Multiply(A=Multiply(A=@n8, B=@n9), B=Multiply(A=@n10.Result, B=@n11.Result))
   Hardness=ScalarParameter() {parameterName=Sphere bandHardness}
  )
  n13 = ScalarParameter() {defaultValue=3, parameterName=Band Exponent}
  n14 = NamedRerouteUsage()
  n15 = VectorParameter() {parameterName=UV1 Mask Options, channelNames {a=contrast, b=add, g=mul, r=pow}, defaultValue=<rgba 5 1 0 0>}
  n16 = Add(A=Multiply(A=Power(Base=@n14, Exp=@n15.pow) {constExponent=5}, B=@n15.mul), B=@n15.add) {constB=0}
  n17 = NamedRerouteUsage()
  n18 = VectorParameter() {parameterName=UV2 Mask Options, channelNames {a=contrast, b=add, g=mul, r=pow}, defaultValue=<rgba 5 1 0 0>}
  n19 = Multiply(A=Power(Base=OneMinus(None=@n17), Exp=@n18.pow) {constExponent=5}, B=@n18.mul)
  n20 = Multiply(
   A=Multiply(
    A=@n2.Mask
    B=MaterialFunctionCall(In=@n16, Contrast=@n15.contrast) {materialFunction=&CheapContrast}.Result
   )
   B=MaterialFunctionCall(In=Add(A=@n19, B=@n18.add) {constB=0}, Contrast=@n18.contrast) {materialFunction=&CheapContrast}.Result
  )
  n21 = ScalarParameter() {defaultValue=0.1, parameterName=Master Opacity}
  n22 = MaterialFunctionCall() {materialFunction=&LinearGradient}
  n23 = VectorParameter() {parameterName="Mask Options (top/bottom letter hiding)", channelNames {a=contrast, b=add, g=mul, r=pow}, defaultValue=<rgba 1 1 0 0>}
  n24 = Add(A=Multiply(A=Power(Base=@n22.VGradient, Exp=@n23.pow), B=@n23.mul), B=@n23.add)
  n25 = Add(
   A=Multiply(
    A=Multiply(A=@n20, B=@n21)
    B=MaterialFunctionCall(In=@n24, Contrast=@n23.contrast) {materialFunction=&CheapContrast}.Result
   )
   B=Multiply(A=Power(Base=@n12, Exp=@n13) {constExponent=10}, B=ScalarParameter() {defaultValue=1, parameterName=Band OpacityStr})
  )
  n26 = VectorParameter() {parameterName=Camera depth options1, channelNames {a=High, b=Low, g=Fade Offset, r=fade length}, defaultValue=<rgba 1500 1500 0.1 1>}
  n27 = VectorParameter() {parameterName=Camera depth options 2, channelNames {a=High, b=Low, g=Fade Offset, r=fade length}, defaultValue=<rgba 500 500 0 1>}
  n28 = Multiply(
   A=LinearInterpolate(
    A=@n26.Low
    B=@n26.Low
    Alpha=MaterialFunctionCall(Fade Length=@n26.fade length, Fade Offset=@n26.fade length) {materialFunction=&CameraDepthFade}.Result
   )
   B=LinearInterpolate(
    A=@n27.Low
    B=@n27.Low
    Alpha=MaterialFunctionCall(Fade Length=@n27.fade length, Fade Offset=@n27.fade length) {materialFunction=&CameraDepthFade}.Result
   )
  )
  n29 = TextureCoordinate() {coordinateIndex=1}
  n30 = NamedRerouteDeclaration(None=OneMinus(None=MaterialFunctionCall(UV Channel=@n29) {materialFunction=&LinearGradient}.VGradient)) {name=UV 1 grad, nodeColor=<rgba 0.7109 0 1 1>}
  n31 = ScalarParameter() {parameterName=collapse}
  n32 = LinearInterpolate(A=Subtract(A=ActorPositionWS(), B=WorldPosition().XYZ), B=@n2.Final WPO, Alpha=Saturate(None=Add(A=@n30, B=@n31)))
  BlendMaterialAttributes(
   A=MakeMaterialAttributes!(
    BaseColor
    Metallic
    Specular
    Roughness
    Anisotropy
    EmissiveColor=Multiply(A=@n3, B=@n1.V)
    Opacity=Multiply(A=@n25, B=@n28)
    OpacityMask
    Normal
    Tangent
    WorldPositionOffset=NamedRerouteDeclaration(None=@n32) {name=WPO, nodeColor=<rgba 0.5703 1 0 1>}
    SubsurfaceColor
    ClearCoat
    ClearCoatRoughness
    AmbientOcclusion
    Refraction
    CustomizedUVs_0
    CustomizedUVs_1
    CustomizedUVs_2
    CustomizedUVs_3
    CustomizedUVs_4
    CustomizedUVs_5
    CustomizedUVs_6
    CustomizedUVs_7
    PixelDepthOffset
    ShadingModel
    Displacement
   )
   B=MakeMaterialAttributes(EmissiveColor=Constant() {r=1}, WorldPositionOffset=@n2.Final WPO)
   Alpha=ScalarParameter() {defaultValue=1, parameterName=Debug}
  )
 ]
}
